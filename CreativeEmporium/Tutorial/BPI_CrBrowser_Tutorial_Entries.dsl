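; BPI_CrBrowser_Tutorial_Entries is a Blueprint interface for entries in the Creative Emporium tutorial browser.
; It declares three functions with no outputs: UpdateEndListPadding takes first and last entry flags, UpdateFocus takes a focus state, and UpdatePanel takes an object reference.
; The file holds signatures only, so the behavior behind each call is left to the implementers.
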(fn UpdateEndListPadding
  (params (IsFirstEntry :type Boolean) (IsLastEntry :type Boolean))
  (return))

(fn UpdateFocus
  (params (IsFocused :type Boolean))
  (return))

(fn UpdatePanel
  (params (Object%20In :type Object%20Reference))
  (return))
