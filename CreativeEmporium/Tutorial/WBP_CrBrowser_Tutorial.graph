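# Full-screen, activatable new-features tutorial for the beta Creative content browser, with vertical tabs beside scrolling panels and input-specific close buttons.
# DataInfo panel objects drive both the page widgets in ScrollBox_Panels and the UIKit view-model tab labels.
# Selecting a tab scrolls to its panel, swaps entry focus through an interface and plays directional sounds.
# Scrolling selects the matching tab, with bFromScroll suppressing scroll-back; gamepad input focuses the first tab.
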WidgetBlueprint /CreativeEmporium/Tutorial/WBP_CrBrowser_Tutorial
graphs [
 EventGraph {nodes [
   n1 = "On Clicked (Button_Close)"!() -> (OutputDelegate, then, Button:Common Button Base Object Reference) {is=title K2Node_ComponentBoundEvent}
   DeactivateWidget(execute=[@n1.then, @n2.then, @n3.then])
   n2 = "On Button Clicked (Button_CloseTouch)"!() -> (OutputDelegate, then, Button:Common Button Legacy Object Reference) {is=title K2Node_ComponentBoundEvent}
   n3 = "On Clicked (Button_Close_Alt)"!() -> (OutputDelegate, then, Button:Common Button Base Object Reference) {is=title K2Node_ComponentBoundEvent}
   n4 = OnTabButtonCreated!() -> (OutputDelegate, then, TabId:Name, TabButton:Common Button Base Object Reference) {is=title K2Node_CustomEvent}
   > n5 = Cast To WBP_UIKit_Button_Selectable!(execute, Object=@n4.TabButton) -> (then, CastFailed, AsWBP UIKit Button Selectable:WBP UIKit Button Selectable Object Reference) {is=title K2Node_DynamicCast}
   > SetPressedSoundOverride!(
    execute
    self:Common Button Base Object Reference=@n5
    Sound:Sound Base Object Reference=Get SelectSound() -> (SelectSound:Sound Base Object Reference)
   ) -> (then)
   > SetHoveredSoundOverride!(
    execute
    self:Common Button Base Object Reference=@n5
    Sound:Sound Base Object Reference=Get HoverSound() -> (HoverSound:Sound Base Object Reference)
   ) -> (then)
   n6 = OnTabSelected!() -> (OutputDelegate, then, TabId:Name) {is=title K2Node_CustomEvent}
   > UpdateFocusedPanel!(
    execute
    self
    Index:int=GetTabIndexByID!(self:UIKit Tab Group Object Reference=Get Button_Tabs() -> (Button_Tabs:WBP UIKit Tabs Object Reference), TabNameID:Name=@n6.TabId) -> (ReturnValue:int=0)
   ) -> (then)
   n7 = OnUserScrolled!() -> (OutputDelegate, then, CurrentOffset:float=~) {is=title K2Node_CustomEvent}
   n8 = Get Button_Tabs()
   n9 = Get PanelHeight() -> (PanelHeight:double)
   n10 = "To Integer (Integer64)"!(InInt:Integer64=Round to Integer64!(A:double=float / float(A=@n7.CurrentOffset, B=@n9)) -> (ReturnValue:Integer64=0)) -> (ReturnValue:int=0)
   > Branch(Condition="Not Equal (Integer)"(A=GetSelectedTabIndex!(self:UIKit Tab Group Object Reference=@n8) -> (ReturnValue:int=0), B=@n10))
   > Set bFromScroll(bFromScroll:bool=true) -> (Output_Get:bool)
   > SelectTabByIndex!(execute, self:UIKit Tab Group Object Reference=@n8, Index:int=@n10, bSuppressClickFeedback:bool=true) -> (then, ReturnValue:bool)
   > Set bFromScroll(bFromScroll=false)
   n11 = OnAthenaScrollFixedPageScrolled!() -> (OutputDelegate, then, ScrollOffset:float=~) {is=title K2Node_CustomEvent}
   > n12 = Branch(Condition="float > float"(A=@n11.ScrollOffset))
   > ScrolledDown()
   ScrolledUp(execute=@n12.else)
   n13 = HandleInputModeChanged!() -> (OutputDelegate, then, bNewInputType:ECommonInputType Enum) {is=title K2Node_CustomEvent}
   > Branch(Condition="Equal (Enum)"(A=@n13.bNewInputType, B=Gamepad))
   > InitializeFocus()
   n14 = Event Construct()
   SetPadding!(
    execute=[@n14.then, @n17.then]
    self:Border Slot Object Reference=SlotAsBorderSlot!(Widget:Widget Object Reference=Get SafeZone_content() -> (SafeZone_content:Safe Zone Object Reference)) -> (ReturnValue:Border Slot Object Reference)
    InPadding_Left:float=0
    InPadding_Top:float=SelectFloat(A=32, B=64, bPickA=IsMobileGame())
    InPadding_Right:float=64
    InPadding_Bottom:float=32
   ) -> (then)
   n15 = Get CommonInputSubsystem()
   > Bind Event to On Input Method Changed(self=@n15, Delegate=@n13)
   > Branch(Condition="Equal (Enum)"(A=GetCurrentInputType(self=@n15), B=Gamepad))
   > DelayUntilNextTick!(execute) -> (then)
   > InitializeFocus()
   Event OnInitialized()
   > Populate Panels()
   > Populate Buttons()
   > UpdateFocusedPanel(Index=Get CurrentIndex() -> (CurrentIndex:int=0))
   > InitializeFocus()
   n16 = Get ScrollBox_Panels() -> (ScrollBox_Panels:Athena Scroll Box Object Reference)
   > Bind Event to On User Scrolled(self:Scroll Box Object Reference=@n16, Delegate=Create Event()) {is=title K2Node_AddDelegate}
   > Bind Event to On Athena Scroll Fixed Page Scrolled(self:Athena Scroll Box Object Reference=@n16, Delegate=Create Event()) {is=title K2Node_AddDelegate}
   Event On Activated()
   > PlaySound2D(Sound=Get AppearSoundCue() -> (AppearSoundCue:Sound Base Object Reference))
   > DelayUntilNextTick()
   > InitializeFocus()
   n17 = Event PreConstruct()
  ]}
 InitializeFocus {nodes [
   InitializeFocus() {is=title K2Node_FunctionEntry}
   n1 = Get Button_Tabs()
   > SetFocus(self=GetChildAt!(
     self:Panel Widget Object Reference=Get TabsContainer!(self:WBP UIKit Tabs Base Object Reference=@n1) -> (TabsContainer:Grid Panel Object Reference)
     Index:int=0
    ) -> (ReturnValue:Widget Object Reference))
  ]}
 Populate Buttons {nodes [
   Populate Buttons() {is=title K2Node_FunctionEntry}
   > n1 = Construct Object from Class(Class=/Script/FortniteUI.UIKitTabGroupViewModel) -> (ReturnValue:UI Kit Tab Group View Model Object Reference)
   > n2 = For Each Loop(Array:Array of BP Cr Browser Tutorial Panel Object References=Get DataInfo() -> (DataInfo:Array of BP Cr Browser Tutorial Panel Object References)) -> (Array Element:BP Cr Browser Tutorial Panel Object Reference)
   Set Tabs!(
    execute=@n2.Completed
    Tabs:Array of UI Kit Tab Group Button View Model Object References=Get TabsDataInfo() -> (TabsDataInfo:Array of UI Kit Tab Group Button View Model Object References)
    self:UI Kit Tab Group View Model Object Reference=@n1
   ) -> (then, Output_Get:Array of UI Kit Tab Group Button View Model Object References)
   n3 = Get Button_Tabs()
   > Set UIKitTabGroupViewModel!(execute, UIKitTabGroupViewModel:UI Kit Tab Group View Model Object Reference=@n1, self:WBP UIKit Tabs Object Reference=@n3) -> (then, Output_Get:UI Kit Tab Group View Model Object Reference)
   > Bind Event to On Tab Selected(self:Common Tab List Widget Base Object Reference=@n3, Delegate=Create Event()) {is=title K2Node_AddDelegate}
   > Bind Event to On Tab Button Creation(self:Common Tab List Widget Base Object Reference=@n3, Delegate=Create Event()) {is=title K2Node_AddDelegate}
   n4 = Construct Object from Class(execute=@n2.LoopBody, Class=/Script/FortniteUI.UIKitTabGroupButtonViewModel) -> (ReturnValue:UI Kit Tab Group Button View Model Object Reference)
   > Set Label!(
    execute
    Label:Text=Get ButtonName!(self:BP Cr Browser Tutorial Panel Object Reference=@n2) -> (ButtonName:Text)
    self:UI Kit Selectable Button View Model Object Reference=@n4
   ) -> (then, Output_Get:Text)
   > Add(
    TargetArray:Array of UI Kit Tab Group Button View Model Object References=Get TabsDataInfo()
    NewItem:"UI Kit Tab Group Button View Model Object Reference (by ref)"=@n4
   ) {is=title K2Node_CallArrayFunction}
  ]}
 Populate Panels {nodes [
   Populate Panels() {is=title K2Node_FunctionEntry}
   > n1 = For Each Loop(Array=Get DataInfo())
   > Set Element(Element:BP Cr Browser Tutorial Panel Object Reference=@n1) -> (Output_Get:BP Cr Browser Tutorial Panel Object Reference)
   > Set Index(Index:int=@n1.Array Index) -> (Output_Get:int=0)
   n2 = Get Panel Type!(self:BP Cr Browser Tutorial Panel Object Reference=Get Element() -> (Element:BP Cr Browser Tutorial Panel Object Reference)) -> (Panel Type:User Widget Class Reference)
   > n3 = Create Widget!(execute:Exec, Class:User Widget Class Reference=@n2, OwningPlayer:Player Controller Object Reference=GetOwningPlayer()) -> (then:Exec, ReturnValue:User Widget Object Reference)
   > n4 = Branch(Condition="Equal (Class)"(A=@n2, B=&/WBP_CrBrowser_Tutorial_PageEntry.WBP_CrBrowser_Tutorial_PageEntry_C))
   > n5 = Cast To WBP_CrBrowser_Tutorial_PageEntry!(execute, Object=@n3) -> (then, CastFailed, AsWBP Cr Browser Tutorial Page Entry:WBP Cr Browser Tutorial Page Entry Object Reference) {is=title K2Node_DynamicCast}
   > UpdatePanel!(execute, self:WBP Cr Browser Tutorial Page Entry Object Reference=@n5, Object In:Object Reference=Get Element()) -> (then)
   > AddChildToVerticalBox(self=Get Vbox_Panels() -> (Vbox_Panels:Vertical Box Object Reference), Content=@n5)
   > UpdatePanel(self=@n5, Object In=Get Element())
   > UpdateFocus!(
    execute
    self:WBP Cr Browser Tutorial Page Entry Object Reference=@n5
    IsFocused:bool="Equal (Integer)"(A=Get Index() -> (Index:int=0), B=Get CurrentIndex() -> (CurrentIndex=0))
   ) -> (then)
   > UpdateEdgePadding!(execute, self, Index:int=Get Index() -> (Index=0), Object:Object Reference=@n5) -> (then)
   n6 = Cast To WBP_CrBrowser_Tutorial_ArtPageEntry!(execute=@n4.else, Object=@n3) -> (then, CastFailed, AsWBP Cr Browser Tutorial Art Page Entry:WBP Cr Browser Tutorial Art Page Entry Object Reference) {is=title K2Node_DynamicCast}
   > UpdatePanel(self:WBP Cr Browser Tutorial Art Page Entry Object Reference=@n6, Object In=Get Element())
   > AddChildToVerticalBox(self=Get Vbox_Panels(), Content=@n6)
   > UpdatePanel(self=@n6, Object In=Get Element())
   > UpdateFocus(
    self:WBP Cr Browser Tutorial Art Page Entry Object Reference=@n6
    IsFocused="Equal (Integer)"(A=Get Index() -> (Index=0), B=Get CurrentIndex() -> (CurrentIndex=0))
   )
   > UpdateEdgePadding(Index=Get Index() -> (Index=0), Object=@n6)
  ]}
 ScrolledDown {nodes [
   ScrolledDown() {is=title K2Node_FunctionEntry}
   > SelectTabByIndex(self=Get Button_Tabs(), Index=int + int(A=Get CurrentIndex() -> (CurrentIndex=0)), bSuppressClickFeedback=true)
  ]}
 ScrolledUp {nodes [
   ScrolledUp() {is=title K2Node_FunctionEntry}
   > SelectTabByIndex(self=Get Button_Tabs(), Index=int - int(A=Get CurrentIndex() -> (CurrentIndex=0)), bSuppressClickFeedback=true)
  ]}
 UpdateEdgePadding {nodes [
   n1 = UpdateEdgePadding!() -> (then, Index:int, Object:Object Reference) {is=title K2Node_FunctionEntry}
   > n2 = Cast To BPI_CrBrowser_Tutorial_Entries!(execute, Object=@n1.Object) -> (then, CastFailed, AsBPI Cr Browser Tutorial Entries:BPI Cr Browser Tutorial Entries Interface) {is=title K2Node_DynamicCast}
   n3 = Get DataInfo()
   > UpdateEndListPadding!(
    execute
    self:BPI Cr Browser Tutorial Entries Interface=@n2
    IsFirstEntry:bool="Equal (Integer)"(A=@n1)
    IsLastEntry:bool="Equal (Integer)"(A=@n1, B=Last Index!(TargetArray:Array of BP Cr Browser Tutorial Panel Object References=@n3) -> (ReturnValue=0))
   ) -> (then)
  ]}
 UpdateFocusedPanel {nodes [
   n1 = UpdateFocusedPanel!() -> (then, Index:int) {is=title K2Node_FunctionEntry}
   > Branch(Condition="Not Equal (Integer)"(A=Get CurrentIndex() -> (CurrentIndex=0), B=@n1))
   n2 = GetChildAt(self=Get Vbox_Panels(), Index=@n1)
   > Is Valid(InputObject=@n2)
   > n3 = Sequence!(execute) -> (then_0, then_1, then_2:Exec, then_3:Exec, then_4:Exec)
   > n4 = Branch(Condition=Get bFromScroll() -> (bFromScroll:bool))
   ScrollWidgetIntoView!(
    execute=@n4.else
    self:Scroll Box Object Reference=Get ScrollBox_Panels()
    WidgetToFind:Widget Object Reference=@n2
    AnimateScroll:bool=true
    ScrollDestination:EDescendantScrollDestination Enum=TopOrLeft
    Padding:float=256
   ) -> (then)
   n5 = Cast To BPI_CrBrowser_Tutorial_Entries(execute=@n3.then_1, Object=@n2)
   > UpdateFocus(self:BPI Cr Browser Tutorial Entries Interface=@n5, IsFocused=true)
   n6 = Cast To BPI_CrBrowser_Tutorial_Entries(execute=@n3.then_2, Object=GetChildAt(self=Get Vbox_Panels(), Index=Get CurrentIndex() -> (CurrentIndex=0)))
   > UpdateFocus(self=@n6, IsFocused=false)
   Branch(execute=@n3.then_3, Condition=IsVisible())
   n7 = Get Index() -> (Index=0)
   n8 = Get CurrentIndex() -> (CurrentIndex=0)
   > PlaySound2D(Sound=Select(
     Option 0:Sound Base Object Reference=Get ScrollUpSoundCue() -> (ScrollUpSoundCue:Sound Base Object Reference)
     Option 1:Sound Base Object Reference=Get ScrollDownSoundCue() -> (ScrollDownSoundCue:Sound Base Object Reference)
     Index="integer > integer"(A=@n7, B=@n8)
    ) -> (ReturnValue:Sound Base Object Reference))
   Set CurrentIndex(execute=@n3.then_4, CurrentIndex:int=Get Index() -> (Index=0)) -> (Output_Get:int=0)
  ]}
]
variables [
 AppearSoundCue {type=SoundBase, default=/Game/Sounds/Fort_UI_Sounds/FrontEnd/New_FrontEnd/Ranked/Fort_UI_Ranked_Whoosh_Intro_Cue}
 bFromScroll {type=boolean}
 CurrentIndex {type=integer}
 DataInfo {
  type="array<BP_CrBrowser_Tutorial_Panel_C>"
  default [
   &/Overview
   &/Filters
   &/FilterModes
   &/ContentArea
   &/FlyOuts
   &/Shortcuts
  ]
 }
 HoverSound {type=SoundBase, default=/Game/Sounds/Fort_UI_Sounds/FrontEnd/New_FrontEnd/Ranked/Fortnite_UI_Ranked_Button_Hover_Cue}
 MainHeaderText {type=string, default=NEW FEATURES TUTORIAL}
 PanelHeight {type=number, default=728}
 Pins {type=Object, default=null}
 ScrollDownSoundCue {type=SoundBase, default=/Game/Sounds/Fort_UI_Sounds/FrontEnd/New_FrontEnd/Ranked/Fort_UI_Shop_Whoosh_Down_Cue}
 ScrollUpSoundCue {type=SoundBase, default=/Game/Sounds/Fort_UI_Sounds/FrontEnd/New_FrontEnd/Ranked/Fort_UI_Ranked_Whoosh_Up_Cue}
 SelectSound {type=SoundBase, default=/Game/Sounds/Fort_UI_Sounds/FrontEnd/New_FrontEnd/Ranked/Fortnite_UI_Ranked_Button_Default_Cue}
 SubHeaderText {type=string, default=BETA CONTENT BROWSER}
 TabsDataInfo {type="array<UIKitTabGroupButtonViewModel>", default []}
]
widgets {
 info {inheritedWidgetCount=0, parentClass=&CommonActivatableWidget, rootWidgetClass=&Border, widgetCount=28}
 widgets [
  Border_0 {
   widgetClassPath=&Border
   widget=Border_0
   properties {
    background {
     bIsDynamicallyLoaded=false
     drawAs=Image
     imageType=FullColor
     mirroring=NoMirror
     resourceName=None
     resourceObject=/CreativeEmporium/Materials/MI_UI_CrBrowser_NebulousBackground
     tiling=NoTile
     imageSize=<32 32>
     margin=<ltrb 0 0 0 0>
     outlineSettings {
      bUseBrushTransparency=false
      roundingType=HalfHeightRadius
      width=0
      color {colorUseRule=UseColor_Specified, specifiedColor=<rgba 0 0 0 0>}
      cornerRadii=<0 0 0 0>
     }
     tintColor {colorUseRule=UseColor_Specified, specifiedColor=<rgba 1 1 1 1>}
     uVRegion {bIsValid=false, max=<0 0>, min=<0 0>}
    }
    padding=<ltrb 0 0 0 0>
   }
  }
  SafeZone_1 {
   widgetClassPath=&SafeZone
   bIsVariable=true
   parent=Border_0
   widget=SafeZone_1
   slot {class=BorderSlot, properties {padding=<ltrb 0 0 0 0>}}
  }
  Overlay_23 {widgetClassPath=&Overlay, parent=SafeZone_1, widget=Overlay_23, slot {class=SafeZoneSlot}}
  WBP_CaptureForPostBufferUpdate {
   widgetClassPath=&WBP_CaptureForPostBufferUpdate
   bIsVariable=true
   parent=Overlay_23
   slot {class=OverlaySlot, properties {horizontalAlignment=HAlign_Fill, verticalAlignment=VAlign_Fill}}
  }
  HorizontalBox_0 {
   widgetClassPath=&HorizontalBox
   parent=Overlay_23
   widget=HorizontalBox_0
   slot {class=OverlaySlot, properties {horizontalAlignment=HAlign_Fill, verticalAlignment=VAlign_Fill}}
  }
  Border_LeftContent {
   widgetClassPath=&CommonBorder
   parent=HorizontalBox_0
   widget=Border_LeftContent
   properties {
    background {
     bIsDynamicallyLoaded=false
     drawAs=Image
     imageType=FullColor
     mirroring=NoMirror
     resourceName=None
     resourceObject=/CreativeEmporium/Materials/MI_CrBrowserTutorialLeftSideBacking
     tiling=NoTile
     imageSize=<64 64>
     margin=<ltrb 0 0 0 0>
     outlineSettings {
      bUseBrushTransparency=false
      roundingType=HalfHeightRadius
      width=0
      color {colorUseRule=UseColor_Specified, specifiedColor=<rgba 0 0 0 0>}
      cornerRadii=<0 0 0 0>
     }
     tintColor {colorUseRule=UseColor_Specified, specifiedColor=<rgba 0.2 0.2 0.2 1>}
     uVRegion {bIsValid=false, max=<0 0>, min=<0 0>}
    }
    padding=<ltrb 0 64 64 32>
   }
   slot {class=HorizontalBoxSlot}
  }
  SafeZone_content {
   widgetClassPath=&SafeZone
   bIsVariable=true
   parent=Border_LeftContent
   widget=SafeZone_content
   properties {padBottom=false, padRight=false, padTop=false}
   slot {class=BorderSlot, properties {padding=<ltrb 0 64 64 32>}}
  }
  LeftSide {
   widgetClassPath=&VerticalBox
   parent=SafeZone_content
   widget=LeftSide
   slot {class=SafeZoneSlot, properties {padding=<ltrb 16 0 0 0>}}
  }
  Spacer_187 {widgetClassPath=&Spacer, parent=LeftSide, widget=Spacer_187, properties {size=<200 200>}, slot {class=VerticalBoxSlot}}
  Button_Tabs {
   widgetClassPath=/FortUILibrary/UIKit/Tabs/WBP_UIKit_Tabs
   bIsVariable=true
   parent=LeftSide
   widget=Button_Tabs
   properties {
    hidePrompts=true
    inputActions=None
    isNaturalNavigation=true
    isVertical=true
    minButtonWidth=300
    navigation=&self:WidgetTree.Button_Tabs.WidgetNavigation_0
    nextPromptHorizontalAlignment=HAlign_Left
    nextPromptPadding=<ltrb 64 0 0 0>
   }
   slot {class=VerticalBoxSlot}
  }
  Spacer_383 {widgetClassPath=&Spacer, parent=HorizontalBox_0, widget=Spacer_383, properties {size=<32 1>}, slot {class=HorizontalBoxSlot}}
  Overlay_2 {
   widgetClassPath=&Overlay
   parent=HorizontalBox_0
   widget=Overlay_2
   slot {class=HorizontalBoxSlot, properties {padding=<ltrb 0 0 64 0>, size {sizeRule=Fill, value=1}}}
  }
  ScrollBox_Panels {
   widgetClassPath=/Script/FortniteUI.AthenaScrollBox
   bIsVariable=true
   parent=Overlay_2
   widget=ScrollBox_Panels
   properties {
    allowOverscroll=false
    bAllowRightClickDragScrolling=false
    bAnimateWheelScrolling=false
    bEnableFixedPageOffset=true
    bEnableTouchScrolling=false
    scrollBarVisibility=Collapsed
    scrollWhenFocusChanges=AnimatedScroll
    visibility=SelfHitTestInvisible
   }
   slot {class=OverlaySlot, properties {horizontalAlignment=HAlign_Fill, verticalAlignment=VAlign_Fill}}
  }
  Vbox_Panels {widgetClassPath=&VerticalBox, bIsVariable=true, parent=ScrollBox_Panels, widget=Vbox_Panels, slot {class=ScrollBoxSlot}}
  SafeZone_0 {
   widgetClassPath=&SafeZone
   bIsVariable=true
   parent=Overlay_2
   widget=SafeZone_0
   properties {padBottom=false, padLeft=false, padRight=false}
   slot {class=OverlaySlot, properties {horizontalAlignment=HAlign_Fill, padding=<ltrb 64 72 92 0>}}
  }
  Border_1 {
   widgetClassPath=&Border
   parent=SafeZone_0
   widget=Border_1
   properties {background {
     bIsDynamicallyLoaded=false
     drawAs=Image
     imageType=FullColor
     mirroring=NoMirror
     resourceName=None
     resourceObject=&MI_CrBrowserTutorialLeftSideBacking
     tiling=NoTile
     imageSize=<32 32>
     margin=<ltrb 0 0 0 0>
     outlineSettings {
      bUseBrushTransparency=false
      roundingType=HalfHeightRadius
      width=0
      color {colorUseRule=UseColor_Specified, specifiedColor=<rgba 0 0 0 0>}
      cornerRadii=<0 0 0 0>
     }
     tintColor {colorUseRule=UseColor_Specified, specifiedColor=<rgba 0.2 0.2 0.2 0>}
     uVRegion {bIsValid=false, max=<0 0>, min=<0 0>}
    }}
   slot {class=SafeZoneSlot}
  }
  Text_Header {widgetClassPath=&VerticalBox, parent=Border_1, widget=Text_Header, slot {class=BorderSlot}}
  HorizontalBox_1 {
   widgetClassPath=&HorizontalBox
   parent=Text_Header
   widget=HorizontalBox_1
   slot {class=VerticalBoxSlot, properties {padding=<ltrb 6 0 0 0>}}
  }
  Text_HeaderSub {
   widgetClassPath=&CommonTextBlock
   bIsVariable=true
   parent=HorizontalBox_1
   widget=Text_HeaderSub
   properties {
    mobileFontSizeMultiplier=1.4
    text=BETA CONTENT BROWSER
    textTransformPolicy=ToUpper
    colorAndOpacity {colorUseRule=UseColor_Specified, specifiedColor=<rgba 0.0314 0.702 1 1>}
    font {
     bForceMonospaced=false
     bMaterialIsStencil=false
     fontMaterial=null
     fontObject=&HeadingNow
     letterSpacing=0
     monospacedWidth=1
     size=18
     skewAmount=0
     typefaceFontName=75 Medium
     outlineSettings {
      bApplyOutlineToDropShadows=false
      bMiteredCorners=false
      bSeparateFillAlpha=false
      outlineMaterial=null
      outlineSize=0
      outlineColor=<rgba 0 0 0 1>
     }
    }
    margin=<ltrb 4 0 4 0>
    shadowOffset=<0 0>
   }
   slot {class=HorizontalBoxSlot}
  }
  Text_HeaderMain {
   widgetClassPath=&CommonTextBlock
   bIsVariable=true
   parent=Text_Header
   widget=Text_HeaderMain
   properties {
    autoWrapText=true
    text=NEW FEATURES TUTORIAL
    textOverflowPolicy=Ellipsis
    textTransformPolicy=ToUpper
    font {
     bForceMonospaced=false
     bMaterialIsStencil=false
     fontMaterial=null
     fontObject=&Burbank
     letterSpacing=0
     monospacedWidth=1
     size=40
     skewAmount=0
     typefaceFontName=BurbankRegular-Black
     outlineSettings {
      bApplyOutlineToDropShadows=false
      bMiteredCorners=false
      bSeparateFillAlpha=false
      outlineMaterial=null
      outlineSize=0
      outlineColor=<rgba 0 0 0 1>
     }
    }
    margin=<ltrb 4 0 4 0>
    shadowOffset=<0 0>
   }
   slot {class=VerticalBoxSlot}
  }
  SafeZone_bottombuttons {
   widgetClassPath=&SafeZone
   bIsVariable=true
   parent=Overlay_23
   widget=SafeZone_bottombuttons
   properties {padLeft=false, padTop=false}
   slot {class=OverlaySlot, properties {horizontalAlignment=HAlign_Right, verticalAlignment=VAlign_Bottom, padding=<ltrb 0 0 36 32>}}
  }
  VisibilityBorder_NonTouch {
   widgetClassPath=&CommonHardwareVisibilityBorder
   parent=SafeZone_bottombuttons
   widget=VisibilityBorder_NonTouch
   properties {visibilityQuery {
     autoDescription=" NONE( Input.Touch )"
     tokenStreamVersion=0
     userDescription=non-touch
     queryTokenStream [0,1,3,1,0]
     tagDictionary [{tagName=Input.Touch}]
    }}
   slot {class=SafeZoneSlot}
  }
  Overlay_3 {
   widgetClassPath=&Overlay
   parent=VisibilityBorder_NonTouch
   widget=Overlay_3
   slot {class=BorderSlot, properties {padding=<ltrb 0 0 0 0>}}
  }
  Button_Close_Alt {
   widgetClassPath=/FortUILibrary/UIKit/Buttons/WBP_UIKit_Button_Quiet
   bIsVariable=true
   parent=Overlay_3
   widget=Button_Close_Alt
   properties {
    bDisplayInputActionWhenNotInteractable=false
    bHideInputAction=true
    bIsFocusable=false
    renderOpacity=0
    text=Close
    hoveredSlateSoundOverride {resourceObject=&Fortnite_UI_Ranked_Button_Hover_Cue}
    pressedSlateSoundOverride {resourceObject=/Game/Sounds/Fort_UI_Sounds/FrontEnd/New_FrontEnd/Ranked/Fortnite_UI_Ranked_Button_Back_Cue}
    selectedHoveredSlateSoundOverride {resourceObject=&Fortnite_UI_Ranked_Button_Hover_Cue}
    selectedPressedSlateSoundOverride {resourceObject=&Fortnite_UI_Ranked_Button_Back_Cue}
    triggeringInputAction {dataTable=/Game/Athena/UI/Frontend/Lobby/AthenaLobbyInputActionDataTable, rowName=HabaneroToggle}
   }
   slot {class=OverlaySlot}
  }
  Button_Close {
   widgetClassPath=&WBP_UIKit_Button_Quiet
   bIsVariable=true
   parent=Overlay_3
   widget=Button_Close
   properties {
    text=Close
    visibility=Visible
    hoveredSlateSoundOverride {resourceObject=&Fortnite_UI_Ranked_Button_Hover_Cue}
    pressedSlateSoundOverride {resourceObject=&Fortnite_UI_Ranked_Button_Back_Cue}
    selectedHoveredSlateSoundOverride {resourceObject=&Fortnite_UI_Ranked_Button_Hover_Cue}
    selectedPressedSlateSoundOverride {resourceObject=&Fortnite_UI_Ranked_Button_Back_Cue}
    triggeringInputAction {dataTable=&AthenaLobbyInputActionDataTable, rowName=HabaneroClose}
   }
   slot {class=OverlaySlot}
  }
  SafeZone_mobileclose {
   widgetClassPath=&SafeZone
   bIsVariable=true
   parent=Overlay_23
   widget=SafeZone_mobileclose
   properties {padBottom=false, padLeft=false}
   slot {class=OverlaySlot, properties {horizontalAlignment=HAlign_Right, padding=<ltrb 0 32 32 0>}}
  }
  VisibilityBorder_Touch {
   widgetClassPath=&CommonHardwareVisibilityBorder
   parent=SafeZone_mobileclose
   widget=VisibilityBorder_Touch
   properties {visibilityQuery {
     autoDescription=" ANY( Input.Touch )"
     tokenStreamVersion=0
     userDescription=touch-only
     queryTokenStream [0,1,1,1,0]
     tagDictionary [{tagName=Input.Touch}]
    }}
   slot {class=SafeZoneSlot}
  }
  Button_CloseTouch {
   widgetClassPath=/FortUILibrary/Foundation/Buttons/CloseButton
   bIsVariable=true
   parent=VisibilityBorder_Touch
   widget=Button_CloseTouch
   properties {
    onButtonBaseClicked=(&self:WidgetTree.Button_CloseTouch.HandleOnButtonClicked)
    onButtonBaseDoubleClicked=(&self:WidgetTree.Button_CloseTouch.HandleOnButtonDoubleClicked)
    onButtonBaseHovered=(&self:WidgetTree.Button_CloseTouch.HandleOnButtonHovered)
    onButtonBaseUnhovered=(&self:WidgetTree.Button_CloseTouch.HandleOnButtonUnhovered)
    onSelectedChangedBase=(&self:WidgetTree.Button_CloseTouch.HandleOnSelectedChanged)
    hoveredSlateSoundOverride {resourceObject=&Fortnite_UI_Ranked_Button_Hover_Cue}
    pressedSlateSoundOverride {resourceObject=&Fortnite_UI_Ranked_Button_Back_Cue}
    selectedHoveredSlateSoundOverride {resourceObject=&Fortnite_UI_Ranked_Button_Hover_Cue}
    selectedPressedSlateSoundOverride {resourceObject=&Fortnite_UI_Ranked_Button_Back_Cue}
   }
   slot {class=BorderSlot, properties {padding=<ltrb 0 0 0 0>}}
  }
 ]
}
