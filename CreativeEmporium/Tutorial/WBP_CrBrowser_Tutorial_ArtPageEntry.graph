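# Art page entry widget for the Creative Emporium browser tutorial, showing a panel image with coach-mark callouts.
# UpdatePanel casts to BP_CrBrowser_Tutorial_Panel, spawns callouts at configured offsets, picks material or texture imagery and tints glow and flare images.
# UpdateFocus plays dim and fade animations, while UpdateEndListPadding sizes spacers for first and last entries.
# PreConstruct mirrors the gradient flares' horizontal shear for Arabic.
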WidgetBlueprint /CreativeEmporium/Tutorial/WBP_CrBrowser_Tutorial_ArtPageEntry
animations [
 Unfocused {
  binding_count=2
  length_seconds=0.2521
  object=&self:Unfocused
  sequence {bindings [
    Border_0 {
     id=#1
     possessed_class=Border
     tracks [{
       type=MovieSceneColorTrack
       display_name=Content Color and Opacity
       sections [{
         type=MovieSceneColorSection
         end_frame=5
         start_frame=0
         channels [
          Color.R {type=MovieSceneScriptingFloatChannel, keys [[0,1],[5,0.1302]]}
          Color.G {type=MovieSceneScriptingFloatChannel, keys [[0,1],[5,0.1302]]}
          Color.B {type=MovieSceneScriptingFloatChannel, keys [[0,1],[5,0.1302]]}
          Color.A {type=MovieSceneScriptingFloatChannel, const [1,2]}
         ]
        }]
      }]
    }
    Overlay_2 {
     id=#2
     possessed_class=Overlay
     tracks [{
       type=MovieSceneFloatTrack
       display_name=Render Opacity
       sections [{
         type=MovieSceneFloatSection
         end_frame=5
         start_frame=0
         channels [None {type=MovieSceneScriptingFloatChannel, keys [[0,1],[0.7497,0],[1.0113,0.0098]]}]
        }]
      }]
    }
   ]}
 }
 Focused {
  binding_count=2
  length_seconds=0.2521
  object=&self:Focused
  sequence {bindings [
    Overlay_2 {
     id=#3
     possessed_class=Overlay
     tracks [{
       type=MovieSceneFloatTrack
       display_name=Render Opacity
       sections [{type=MovieSceneFloatSection, end_frame=5, start_frame=0, channels [None {type=MovieSceneScriptingFloatChannel, keys [[0,0],[5,1]]}]}]
      }]
    }
    Border_0 {
     id=#1
     possessed_class=Border
     tracks [{
       type=MovieSceneColorTrack
       display_name=Content Color and Opacity
       sections [{
         type=MovieSceneColorSection
         end_frame=5
         start_frame=0
         channels [
          Color.R {type=MovieSceneScriptingFloatChannel, keys [[0,0.1302],[5,1],[5.0427,1]]}
          Color.G {type=MovieSceneScriptingFloatChannel, keys [[0,0.1302],[5,1],[5.0427,1]]}
          Color.B {type=MovieSceneScriptingFloatChannel, keys [[0,0.1302],[5,1],[5.0427,1]]}
          Color.A {type=MovieSceneScriptingFloatChannel, const [1,3]}
         ]
        }]
      }]
    }
   ]}
 }
]
graphs [
 CreateCallout {nodes [
   n1 = CreateCallout!() -> (then, CrBrowser_Tutorial_PanelCallouts:"Cr Browser Tutorial Panel Callouts Structure (by ref)") {is=title K2Node_FunctionEntry}
   > n2 = Create Widget!(
    execute:Exec
    Class:User Widget Class Reference=/FortUILibrary/Systems/CoachMark/WBP_UIKit_CoachMark_Base.WBP_UIKit_CoachMark_Base_C
    OwningPlayer:Player Controller Object Reference=GetOwningPlayer()
   ) -> (then:Exec, ReturnValue:WBP UIKit Coach Mark Base Object Reference)
   n3 = Break Cr Browser Tutorial Panel Callouts!(CrBrowser_Tutorial_PanelCallouts:"Cr Browser Tutorial Panel Callouts Structure (by ref)"=@n1) -> (
    TextDescription_21_#4:Text
    TextHeader_24_#5:Text
    ShowArrow_1_#6:bool=~
    ArrowDirection_5_#7:E_UI_Alignment Enum
    WidthOverride_8_#8:double=~
    HeightOverride_10_#9:double=~
    TextJustification_13_#10:ETextJustify Enum
    Offset_17_#11:Vector 2D Structure
   ) {is=title K2Node_BreakStruct}
   > Set TextDescription!(execute, TextDescription:Text=@n3, self:WBP UIKit Coach Mark Base Object Reference=@n2) -> (then, Output_Get:Text)
   n4 = GetEmptyText!() -> (ReturnValue:Text)
   > SetVisibility(
    self=Get RichText_Header!(self:WBP UIKit Coach Mark Base Object Reference=@n2) -> (RichText_Header:Common Rich Text Block Object Reference)
    InVisibility=Select(
     Option 0=SelfHitTestInvisible
     Option 1=Collapsed
     Index="Equal Exactly (Text)"!(A:"Text (by ref)"=@n3.TextHeader_24_#5, B:"Text (by ref)"=@n4) -> (ReturnValue:bool)
    )
   )
   > Set TextHeader!(execute, TextHeader:Text=@n3.TextHeader_24_#5, self:WBP UIKit Coach Mark Base Object Reference=@n2) -> (then, Output_Get:Text)
   > Set ArrowPosition!(
    execute
    ArrowPosition:E_UI_OutBoardPosition Enum=Select(
     Option 0:E_UI_OutBoardPosition Enum=NewEnumerator3
     Option 1:E_UI_OutBoardPosition Enum=Select!(
      NewEnumerator0:E_UI_OutBoardPosition Enum=NewEnumerator1
      NewEnumerator1:E_UI_OutBoardPosition Enum=NewEnumerator5
      NewEnumerator2:E_UI_OutBoardPosition Enum=NewEnumerator8
      NewEnumerator3:E_UI_OutBoardPosition Enum=NewEnumerator11
      Index:E_UI_Alignment Enum=@n3.ArrowDirection_5_#7
     ) -> (ReturnValue)
     Index=@n3.ShowArrow_1_#6
    ) -> (ReturnValue:E_UI_OutBoardPosition Enum)
    self:WBP UIKit Coach Mark Base Object Reference=@n2
   ) -> (then, Output_Get:E_UI_OutBoardPosition Enum=NewEnumerator3)
   > n5 = Branch(Condition="float > float"(A=@n3.WidthOverride_8_#8))
   > n6 = Set Widget Width!(execute, Widget Width:float=@n3.WidthOverride_8_#8, self:WBP UIKit Coach Mark Base Object Reference=@n2) -> (then, Output_Get:float)
   Set ShouldShowDismissButton!(execute=[@n5.else, @n6.then], ShouldShowDismissButton:bool=false, self:WBP UIKit Coach Mark Base Object Reference=@n2) -> (then, Output_Get:bool)
   > Set OtherActionButtonClass!(execute, OtherActionButtonClass:WBP UIKit Button Regular Soft Class Reference, self:WBP UIKit Coach Mark Base Object Reference=@n2) -> (then, Output_Get:WBP UIKit Button Regular Soft Class Reference)
   > Set JustificationDescription!(
    execute
    JustificationDescription:ETextJustify Enum=@n3.TextJustification_13_#10
    self:WBP UIKit Coach Mark Base Object Reference=@n2
   ) -> (then, Output_Get:ETextJustify Enum=Left)
   > Set JustificationHeader!(
    execute
    JustificationHeader:ETextJustify Enum=@n3.TextJustification_13_#10
    self:WBP UIKit Coach Mark Base Object Reference=@n2
   ) -> (then, Output_Get:ETextJustify Enum=Left)
   > Set DefaultDescriptoinColor!(
    execute
    DefaultDescriptoinColor:Slate Color Structure=Make Slate Color(SpecifiedColor=(R=1,G=0.808,B=0,A=1))
    self:WBP UIKit Coach Mark Base Object Reference=@n2
   ) -> (then, Output_Get:Slate Color Structure)
   > SetVisibility(self=@n2, InVisibility=SelfHitTestInvisible)
   > Return Node!(
    execute
    CalloutWidget:WBP UIKit Coach Mark Base Object Reference=@n2
    WidgetTransform:Widget Transform Structure=Make Widget Transform!(
     Translation:Vector 2D Structure=@n3.Offset_17_#11
     Scale:Vector 2D Structure=(X=1,Y=1)
     Shear:Vector 2D Structure=(X=0,Y=0)
     Angle:float=0
    ) -> (WidgetTransform:Widget Transform Structure) {is=title K2Node_MakeStruct}
   )
  ]}
 EventGraph {nodes [
   Event PreConstruct()
   n1 = Get GradientFlare() -> (GradientFlare:Image Object Reference)
   n2 = Break Widget Transform!(WidgetTransform:"Widget Transform Structure (by ref)"=Get RenderTransform!(self:Widget Object Reference=@n1) -> (RenderTransform:Widget Transform Structure)) -> (Translation:Vector 2D Structure, Scale:Vector 2D Structure, Shear:Vector 2D Structure, Angle:float=~) {is=title K2Node_BreakStruct}
   n3 = "Equal, Case Insensitive (String)"(A=GetCurrentLanguage!() -> (ReturnValue:String), B=ar)
   n4 = BreakVector2D(InVec=@n2.Shear)
   n5 = float * float(A=@n4, B=-1) {is=title K2Node_PromotableOperator}
   > SetRenderShear!(
    execute
    self:Widget Object Reference=@n1
    Shear:Vector 2D Structure=Select!(Option 0:Vector 2D Structure=@n2.Shear, Option 1:Vector 2D Structure=MakeVector2D(X=@n5, Y=@n4.Y), Index:bool=@n3) -> (ReturnValue:Vector 2D Structure)
   ) -> (then)
   n6 = Get GradientFlare_Bg() -> (GradientFlare_Bg:Image Object Reference)
   n7 = Break Widget Transform(WidgetTransform=Get RenderTransform(self=@n6)) -> (Angle=~)
   n8 = BreakVector2D(InVec=@n7.Shear)
   n9 = float * float(A=@n8, B=-1)
   > SetRenderShear(self=@n6, Shear=Select(Option 0=@n7.Shear, Option 1=MakeVector2D(X=@n9, Y=@n8.Y), Index=@n3))
   n10 = Event UpdatePanel!() -> (OutputDelegate, then, Object In:Object Reference) {is=title K2Node_Event}
   > UpdateCallouts!(execute, self, Object:Object Reference=@n10.Object In) -> (then)
   > UpdateImage!(execute, self, Object:Object Reference=@n10.Object In) -> (then)
   > UpdatePanelColor!(execute, self, Object:Object Reference=@n10.Object In) -> (then)
   n11 = Event UpdateFocus!() -> (OutputDelegate, then, IsFocused:bool) {is=title K2Node_Event}
   > PlayAnimation(InAnimation=Select(
     Option 0:Widget Animation Object Reference=Get Unfocused() -> (Unfocused:Widget Animation Object Reference)
     Option 1:Widget Animation Object Reference=Get Focused() -> (Focused:Widget Animation Object Reference)
     Index=@n11.IsFocused
    ) -> (ReturnValue:Widget Animation Object Reference))
   n12 = Event UpdateEndListPadding!() -> (OutputDelegate, then, IsFirstEntry:bool, IsLastEntry:bool) {is=title K2Node_Event}
   > SetSize!(
    execute
    self:Spacer Object Reference=Get Spacer_Top() -> (Spacer_Top:Spacer Object Reference)
    InSize_X:double=0
    InSize_Y:double=SelectFloat(A=256, B=0, bPickA=@n12.IsFirstEntry)
   ) -> (then)
   > SetSize(self=Get Spacer_Bottom() -> (Spacer_Bottom:Spacer Object Reference), InSize_X=0, InSize_Y=SelectFloat(A=1000, B=0, bPickA=@n12.IsLastEntry))
  ]}
 UpdateCallouts {nodes [
   n1 = Reroute Node(InputPin:BP Cr Browser Tutorial Panel Object Reference) -> (OutputPin:BP Cr Browser Tutorial Panel Object Reference)
   "SetText (Text)"(
    self=Get Text_Title() -> (Text_Title:Common Text Object Reference)
    InText=Get Title!(self:BP Cr Browser Tutorial Panel Object Reference=@n1) -> (Title:Text)
   )
   n2 = Get Details List!(self:BP Cr Browser Tutorial Panel Object Reference=@n1) -> (Details List:Array of Texts)
   > SetText!(
    execute
    self:WBP Cr Browser Tutorial Bullet Entry Object Reference=Get WBP_CrBrowser_Tutorial_BulletEntry() -> (WBP_CrBrowser_Tutorial_BulletEntry:WBP Cr Browser Tutorial Bullet Entry Object Reference)
    In Text:Text="Get (a copy)"(Array:Array of Texts=@n2) -> (Output:Text)
   ) -> (then)
   > SetText(
    self=Get WBP_CrBrowser_Tutorial_BulletEntry_1() -> (WBP_CrBrowser_Tutorial_BulletEntry_1:WBP Cr Browser Tutorial Bullet Entry Object Reference)
    In Text="Get (a copy)"(Array=@n2, Dimension 1=1)
   )
   n3 = UpdateCallouts!() -> (then, Object:Object Reference) {is=title K2Node_FunctionEntry}
   > n4 = Cast To BP_CrBrowser_Tutorial_Panel!(execute, Object=@n3) -> (then, CastFailed, AsBP Cr Browser Tutorial Panel:BP Cr Browser Tutorial Panel Object Reference) {is=title K2Node_DynamicCast}
   n5 = Get DetailsCallouts!(self:BP Cr Browser Tutorial Panel Object Reference=@n4) -> (DetailsCallouts:Array of Cr Browser Tutorial Panel Callouts Structures)
   > Branch(Condition="integer > integer"(A=Length(TargetArray:Array of Cr Browser Tutorial Panel Callouts Structures=@n5)))
   > n6 = For Each Loop(Array:Array of Cr Browser Tutorial Panel Callouts Structures=@n5) -> (Array Element:Cr Browser Tutorial Panel Callouts Structure)
   > n7 = CreateCallout!(execute, self, CrBrowser_Tutorial_PanelCallouts:"Cr Browser Tutorial Panel Callouts Structure (by ref)"=@n6) -> (then, CalloutWidget:WBP UIKit Coach Mark Base Object Reference, WidgetTransform:Widget Transform Structure) {is=title K2Node_CallFunction}
   > AddChildToOverlay!(
    execute
    self:Overlay Object Reference=Get CalloutsContainer() -> (CalloutsContainer:Overlay Object Reference)
    Content:Widget Object Reference=@n7
   ) -> (then, ReturnValue:Overlay Slot Object Reference)
   > Set RenderTransform!(execute, RenderTransform:Widget Transform Structure=@n7.WidgetTransform, self:Widget Object Reference=@n7) -> (then, Output_Get:Widget Transform Structure)
  ]}
 UpdateImage {nodes [
   n1 = UpdateImage!() -> (then, Object:Object Reference) {is=title K2Node_FunctionEntry}
   > n2 = Cast To BP_CrBrowser_Tutorial_Panel(Object=@n1)
   > n3 = Sequence()
   n4 = Get ImageBG!(self:BP Cr Browser Tutorial Panel Object Reference=@n2) -> (ImageBG:Material Instance Object Reference)
   n5 = Branch(execute=@n3.then_1, Condition=IsValid(Object=@n4))
   n6 = Get IconBG() -> (IconBG:Image Object Reference)
   > SetBrushFromMaterial(self=@n6, Material=@n4)
   > SetVisibility(self=[Get Image_Content_R() -> (Image_Content_R:Scale Box Object Reference), @n6], InVisibility=HitTestInvisible)
   SetVisibility(execute=@n5.else, self=Get IconBG(), InVisibility=Collapsed)
   SetVisibility(
    execute=@n3.then_0
    self=[Get IconBG(), Get Icon() -> (Icon:Image Object Reference), Get Texture() -> (Texture:Image Object Reference)]
    InVisibility=Collapsed
   )
   n7 = Get Image!(self:BP Cr Browser Tutorial Panel Object Reference=@n2) -> (Image:Material Instance Object Reference)
   > n8 = Branch(Condition=IsValid(Object=@n7))
   > SetBrushFromMaterial(self=Get Icon(), Material=@n7)
   > SetVisibility(self=[Get Icon(), Get Image_Content_R()], InVisibility=SelfHitTestInvisible)
   n9 = Get TextureImage!(self:BP Cr Browser Tutorial Panel Object Reference=@n2) -> (TextureImage:Texture 2D Object Reference)
   Branch(execute=@n8.else, Condition=IsValid(Object=@n9))
   > SetBrushFromTexture!(execute, self:Image Object Reference=Get Texture(), Texture:Texture 2D Object Reference=@n9, bMatchSize:bool=false) -> (then)
   > SetVisibility(self=[Get Texture(), Get Image_Content_R()], InVisibility=HitTestInvisible)
  ]}
 UpdatePanelColor {nodes [
   n1 = UpdatePanelColor!() -> (then, Object:Object Reference) {is=title K2Node_FunctionEntry}
   > n2 = Cast To BP_CrBrowser_Tutorial_Panel(Object=@n1)
   n3 = Break Cr Browser Tutorial Panel Colors!(CrBrowser_Tutorial_PanelColors:"Cr Browser Tutorial Panel Colors Structure (by ref)"=Get PanelColors!(self:BP Cr Browser Tutorial Panel Object Reference=@n2) -> (PanelColors:Cr Browser Tutorial Panel Colors Structure)) -> (
    Glow_19_#12:Linear Color Structure
    Flare_ColorA_14_#13:Linear Color Structure
    Flare_ColorB_15_#14:Linear Color Structure
    FlareBG_ColorA_16_#15:Linear Color Structure
    FlareBG_ColorB_17_#16:Linear Color Structure
   ) {is=title K2Node_BreakStruct}
   > SetColorAndOpacity(self:Image Object Reference=Get Glow() -> (Glow:Image Object Reference), InColorAndOpacity:Linear Color Structure=@n3)
   > n4 = GetDynamicMaterial(self=Get GradientFlare())
   > SetVectorParameterValue(self=@n4, ParameterName=Color A, Value=@n3.Flare_ColorA_14_#13)
   > SetVectorParameterValue(self=@n4, ParameterName=Color B, Value=@n3.Flare_ColorB_15_#14)
   > n5 = GetDynamicMaterial(self=Get GradientFlare_Bg())
   > SetVectorParameterValue(self=@n5, ParameterName=Color A, Value=@n3.FlareBG_ColorA_16_#15)
   > SetVectorParameterValue(self=@n5, ParameterName=Color B, Value=@n3.FlareBG_ColorB_17_#16)
  ]}
]
variables []
widgets {
 info {inheritedWidgetCount=0, parentClass=&CommonActivatableWidget, rootWidgetClass=&Border, widgetCount=23}
 widgets [
  Border_0 {
   widgetClassPath=&Border
   widget=Border_0
   properties {
    background {
     bIsDynamicallyLoaded=false
     drawAs=NoDrawType
     imageType=NoImage
     mirroring=NoMirror
     resourceName=None
     resourceObject=null
     tiling=NoTile
     imageSize=<32 32>
     margin=<ltrb 0 0 0 0>
     outlineSettings {
      bUseBrushTransparency=false
      roundingType=HalfHeightRadius
      width=0
      color {colorUseRule=UseColor_Specified, specifiedColor=<rgba 0 0 0 0>}
      cornerRadii=<0 0 0 0>
     }
     tintColor {colorUseRule=UseColor_Specified, specifiedColor=<rgba 1 1 1 1>}
     uVRegion {bIsValid=false, max=<0 0>, min=<0 0>}
    }
    padding=<ltrb 0 32 0 32>
   }
  }
  VerticalBox_1 {
   widgetClassPath=&VerticalBox
   parent=Border_0
   widget=VerticalBox_1
   slot {class=BorderSlot, properties {padding=<ltrb 0 32 0 32>}}
  }
  Spacer_Top {widgetClassPath=&Spacer, bIsVariable=true, parent=VerticalBox_1, widget=Spacer_Top, slot {class=VerticalBoxSlot}}
  Overlay_Modal {
   widgetClassPath=&Overlay
   parent=VerticalBox_1
   widget=Overlay_Modal
   slot {class=VerticalBoxSlot, properties {size {sizeRule=Fill, value=1}}}
  }
  Image_Backing {
   widgetClassPath=&Image
   bIsVariable=true
   parent=Overlay_Modal
   widget=Image_Backing
   properties {brush {
     bIsDynamicallyLoaded=false
     drawAs=RoundedBox
     imageType=FullColor
     mirroring=NoMirror
     resourceName=None
     resourceObject=/CreativeEmporium/Materials/MI_CrBrowserTutorialBackplate
     tiling=NoTile
     imageSize=<32 32>
     margin=<ltrb 0 0 0 0>
     outlineSettings {
      bUseBrushTransparency=false
      roundingType=FixedRadius
      width=0
      color {colorUseRule=UseColor_Specified, specifiedColor=<rgba 0 0 0 0>}
      cornerRadii=<40 40 40 40>
     }
     tintColor {colorUseRule=UseColor_Specified, specifiedColor=<rgba 1 1 1 1>}
     uVRegion {bIsValid=false, max=<0 0>, min=<0 0>}
    }}
   slot {class=OverlaySlot, properties {horizontalAlignment=HAlign_Fill, verticalAlignment=VAlign_Fill}}
  }
  PanelSize {
   widgetClassPath=&Spacer
   parent=Overlay_Modal
   widget=PanelSize
   properties {size=<664 664>}
   slot {class=OverlaySlot, properties {horizontalAlignment=HAlign_Fill, verticalAlignment=VAlign_Fill}}
  }
  Overlay_2 {
   widgetClassPath=&Overlay
   parent=Overlay_Modal
   widget=Overlay_2
   slot {class=OverlaySlot, properties {horizontalAlignment=HAlign_Fill, verticalAlignment=VAlign_Fill}}
  }
  Image_Content_R {
   widgetClassPath=&ScaleBox
   bIsVariable=true
   parent=Overlay_2
   widget=Image_Content_R
   properties {stretch=ScaleBySafeZone, stretchDirection=DownOnly, visibility=Collapsed}
   slot {class=OverlaySlot, properties {horizontalAlignment=HAlign_Fill, verticalAlignment=VAlign_Center}}
  }
  SafeZone_7 {
   widgetClassPath=&SafeZone
   bIsVariable=true
   parent=Image_Content_R
   widget=SafeZone_7
   properties {padBottom=false, padLeft=false, padTop=false}
   slot {class=ScaleBoxSlot, properties {horizontalAlignment=HAlign_Fill}}
  }
  Overlay_0 {widgetClassPath=&Overlay, parent=SafeZone_7, widget=Overlay_0, slot {class=SafeZoneSlot}}
  Glow {
   widgetClassPath=&Image
   bIsVariable=true
   parent=Overlay_0
   widget=Glow
   properties {
    visibility=Hidden
    brush {
     bIsDynamicallyLoaded=false
     drawAs=Image
     imageType=FullColor
     mirroring=NoMirror
     resourceName=None
     resourceObject=null
     tiling=NoTile
     imageSize=<32 32>
     margin=<ltrb 0 0 0 0>
     outlineSettings {
      bUseBrushTransparency=false
      roundingType=HalfHeightRadius
      width=0
      color {colorUseRule=UseColor_Specified, specifiedColor=<rgba 0 0 0 0>}
      cornerRadii=<0 0 0 0>
     }
     tintColor {colorUseRule=UseColor_Specified, specifiedColor=<rgba 1 1 1 1>}
     uVRegion {bIsValid=false, max=<0 0>, min=<0 0>}
    }
    renderTransform {angle=0, scale=<32 32>, shear=<0 0>, translation=<0 0>}
   }
   slot {class=OverlaySlot, properties {horizontalAlignment=HAlign_Right}}
  }
  GradientFlare_Bg {
   widgetClassPath=&Image
   bIsVariable=true
   parent=Overlay_0
   widget=GradientFlare_Bg
   properties {visibility=Hidden, renderTransform {angle=0, scale=<1 1>, shear=<24 0>, translation=<0 0>}}
   slot {class=OverlaySlot, properties {horizontalAlignment=HAlign_Fill, verticalAlignment=VAlign_Fill}}
  }
  GradientFlare {
   widgetClassPath=&Image
   bIsVariable=true
   parent=Overlay_0
   widget=GradientFlare
   properties {visibility=Hidden, renderTransform {angle=0, scale=<0.875 1>, shear=<24 0>, translation=<0 0>}}
   slot {class=OverlaySlot, properties {horizontalAlignment=HAlign_Fill, verticalAlignment=VAlign_Fill}}
  }
  IconBG {
   widgetClassPath=&Image
   bIsVariable=true
   parent=Overlay_0
   widget=IconBG
   properties {
    visibility=HitTestInvisible
    brush {
     bIsDynamicallyLoaded=false
     drawAs=Image
     imageType=NoImage
     mirroring=NoMirror
     resourceName=None
     resourceObject=null
     tiling=NoTile
     imageSize=<600 600>
     margin=<ltrb 0 0 0 0>
     outlineSettings {
      bUseBrushTransparency=false
      roundingType=HalfHeightRadius
      width=0
      color {colorUseRule=UseColor_Specified, specifiedColor=<rgba 0 0 0 0>}
      cornerRadii=<0 0 0 0>
     }
     tintColor {colorUseRule=UseColor_Specified, specifiedColor=<rgba 1 1 1 1>}
     uVRegion {bIsValid=false, max=<0 0>, min=<0 0>}
    }
   }
   slot {class=OverlaySlot, properties {horizontalAlignment=HAlign_Fill, verticalAlignment=VAlign_Center, padding=<ltrb 32 32 32 32>}}
  }
  Icon {
   widgetClassPath=&Image
   bIsVariable=true
   parent=Overlay_0
   widget=Icon
   properties {
    visibility=HitTestInvisible
    brush {
     bIsDynamicallyLoaded=false
     drawAs=Image
     imageType=NoImage
     mirroring=NoMirror
     resourceName=None
     resourceObject=null
     tiling=NoTile
     imageSize=<64 600>
     margin=<ltrb 0 0 0 0>
     outlineSettings {
      bUseBrushTransparency=false
      roundingType=HalfHeightRadius
      width=0
      color {colorUseRule=UseColor_Specified, specifiedColor=<rgba 0 0 0 0>}
      cornerRadii=<0 0 0 0>
     }
     tintColor {colorUseRule=UseColor_Specified, specifiedColor=<rgba 1 1 1 1>}
     uVRegion {bIsValid=false, max=<0 0>, min=<0 0>}
    }
   }
   slot {class=OverlaySlot, properties {horizontalAlignment=HAlign_Fill, verticalAlignment=VAlign_Fill, padding=<ltrb 32 32 32 32>}}
  }
  ScaleBox_0 {
   widgetClassPath=&ScaleBox
   parent=Overlay_0
   widget=ScaleBox_0
   properties {stretch=ScaleToFitX}
   slot {class=OverlaySlot, properties {horizontalAlignment=HAlign_Center, verticalAlignment=VAlign_Center, padding=<ltrb 12 12 12 12>}}
  }
  Overlay_3 {widgetClassPath=&Overlay, parent=ScaleBox_0, widget=Overlay_3, slot {class=ScaleBoxSlot}}
  SizeBox_0 {
   widgetClassPath=&SizeBox
   parent=Overlay_3
   widget=SizeBox_0
   properties {bOverride_HeightOverride=true, bOverride_WidthOverride=true, heightOverride=600, widthOverride=1100}
   slot {class=OverlaySlot, properties {horizontalAlignment=HAlign_Center, verticalAlignment=VAlign_Center}}
  }
  Texture {
   widgetClassPath=&Image
   bIsVariable=true
   parent=SizeBox_0
   widget=Texture
   properties {brush {
     bIsDynamicallyLoaded=false
     drawAs=Image
     imageType=FullColor
     mirroring=NoMirror
     resourceName=None
     resourceObject=null
     tiling=NoTile
     imageSize=<1332 723>
     margin=<ltrb 0 0 0 0>
     outlineSettings {
      bUseBrushTransparency=false
      roundingType=HalfHeightRadius
      width=0
      color {colorUseRule=UseColor_Specified, specifiedColor=<rgba 0 0 0 0>}
      cornerRadii=<0 0 0 0>
     }
     tintColor {colorUseRule=UseColor_Specified, specifiedColor=<rgba 1 1 1 1>}
     uVRegion {bIsValid=false, max=<0 0>, min=<0 0>}
    }}
   slot {class=SizeBoxSlot, properties {horizontalAlignment=HAlign_Center, verticalAlignment=VAlign_Center}}
  }
  CalloutsContainer {
   widgetClassPath=&Overlay
   bIsVariable=true
   parent=Overlay_3
   widget=CalloutsContainer
   slot {class=OverlaySlot, properties {horizontalAlignment=HAlign_Fill, verticalAlignment=VAlign_Fill}}
  }
  Overlay_1 {
   widgetClassPath=&Overlay
   parent=Overlay_2
   widget=Overlay_1
   slot {class=OverlaySlot, properties {horizontalAlignment=HAlign_Center, verticalAlignment=VAlign_Center}}
  }
  Text_Title {
   widgetClassPath=&CommonTextBlock
   bIsVariable=true
   parent=Overlay_1
   widget=Text_Title
   properties {
    style=/FortUILibrary/Styles/Typography/_TextStyle_SubH4._TextStyle_SubH4_C
    textTransformPolicy=ToUpper
    wrapTextAt=1152
    font {
     bForceMonospaced=false
     bMaterialIsStencil=false
     fontMaterial=null
     fontObject=&BurbankSmall
     letterSpacing=0
     monospacedWidth=1
     size=24
     skewAmount=0
     typefaceFontName=Black
     outlineSettings {
      bApplyOutlineToDropShadows=false
      bMiteredCorners=false
      bSeparateFillAlpha=false
      outlineMaterial=null
      outlineSize=0
      outlineColor=<rgba 0 0 0 1>
     }
    }
    shadowOffset=<0 0>
    strikeBrush {
     bIsDynamicallyLoaded=false
     drawAs=NoDrawType
     imageType=NoImage
     mirroring=NoMirror
     resourceName=None
     resourceObject=null
     tiling=NoTile
     imageSize=<32 32>
     margin=<ltrb 0 0 0 0>
     outlineSettings {
      bUseBrushTransparency=false
      roundingType=HalfHeightRadius
      width=0
      color {colorUseRule=UseColor_Specified, specifiedColor=<rgba 0 0 0 0>}
      cornerRadii=<0 0 0 0>
     }
     tintColor {colorUseRule=UseColor_Specified, specifiedColor=<rgba 1 1 1 1>}
     uVRegion {bIsValid=false, max=<0 0>, min=<0 0>}
    }
   }
   slot {class=OverlaySlot, properties {horizontalAlignment=HAlign_Fill, padding=<ltrb 24 24 24 24>}}
  }
  Spacer_Bottom {widgetClassPath=&Spacer, bIsVariable=true, parent=VerticalBox_1, widget=Spacer_Bottom, slot {class=VerticalBoxSlot}}
 ]
}
