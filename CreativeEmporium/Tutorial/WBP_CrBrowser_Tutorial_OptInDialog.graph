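# Opt-in dialog for the Creative Emporium browser tutorial, an activatable panel wrapping the UIKit internal dialog with a character image.
# It feeds the dialog a fresh opt-in viewmodel on design-time preview, construct and activation, when an Intro animation slides and fades the image in.
# Button clicks broadcast SelectionMadeEvent, with Confirmed true only for index 2, then deactivate the panel.
# The click binding is made on Construct but cleared on deactivation.
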WidgetBlueprint /CreativeEmporium/Tutorial/WBP_CrBrowser_Tutorial_OptInDialog
animations [Intro {
  binding_count=1
  length_seconds=0.5
  object=&self:Intro
  sequence {bindings [Image {
     id=#1
     possessed_class=Image
     tracks [
      {
       type=MovieScene2DTransformTrack
       sections [{
         type=MovieScene2DTransformSection
         end_frame=10
         start_frame=0
         channels [
          Translation.X {type=MovieSceneScriptingFloatChannel, const [0,1]}
          Translation.Y {type=MovieSceneScriptingFloatChannel, keys [[0,64],[10,0]]}
          Angle {type=MovieSceneScriptingFloatChannel}
          Scale.X {type=MovieSceneScriptingFloatChannel}
          Scale.Y {type=MovieSceneScriptingFloatChannel}
          Shear.X {type=MovieSceneScriptingFloatChannel}
          Shear.Y {type=MovieSceneScriptingFloatChannel}
         ]
        }]
      }
      {
       type=MovieSceneFloatTrack
       display_name=Render Opacity
       sections [{type=MovieSceneFloatSection, end_frame=10, start_frame=0, channels [None {type=MovieSceneScriptingFloatChannel, keys [[0,0],[10,1]]}]}]
      }
     ]
    }]}
 }]
graphs [
 EventGraph {nodes [
   n1 = HandleClick!() -> (OutputDelegate, then, ButtonIndex:int) {is=title K2Node_CustomEvent}
   > Call Selection Made Event!(
    execute:Exec
    self:WBP Cr Browser Tutorial Opt In Dialog Object Reference
    Confirmed:bool=Select!(Option 0:bool=false, Option 1:bool=false, Option 2:bool=true, Index:int=@n1.ButtonIndex) -> (ReturnValue:bool)
   ) -> (then:Exec) {is=title K2Node_CallDelegate}
   > DeactivateWidget()
   n2 = Event PreConstruct()
   > Branch(Condition=@n2.IsDesignTime)
   > n3 = Construct Object from Class(Class=&/DialogVM_CrBrowserTutorial_OptIn.DialogVM_CrBrowserTutorial_OptIn_C) -> (ReturnValue:Dialog VM Cr Browser Tutorial Opt In Object Reference)
   > SetDialogData!(
    execute
    self:WBP UIKit Dialog Internal Object Reference=Get WBP_UIKit_Dialog_Base() -> (WBP_UIKit_Dialog_Base:WBP UIKit Dialog Internal Object Reference)
    Data:UI Kit Dialog Viewmodel Object Reference=@n3
   ) -> (then)
   Event On Activated()
   > n4 = Construct Object from Class(Class=&DialogVM_CrBrowserTutorial_OptIn_C)
   > SetDialogData(self=Get WBP_UIKit_Dialog_Base(), Data=@n4)
   > PlayAnimation(InAnimation=Get Intro())
   Event On Deactivated()
   > Unbind all Events from Dialog Button Clicked Dispatcher!(execute:Exec, self:WBP UIKit Dialog Internal Object Reference=Get WBP_UIKit_Dialog_Base()) -> (then:Exec) {is=title K2Node_ClearDelegate}
   Event Construct()
   > n5 = Construct Object from Class(Class=&DialogVM_CrBrowserTutorial_OptIn_C)
   > SetDialogData(self=Get WBP_UIKit_Dialog_Base(), Data=@n5)
   > Bind Event to Dialog Button Clicked Dispatcher(self:WBP UIKit Dialog Internal Object Reference=Get WBP_UIKit_Dialog_Base(), Delegate=@n1) {is=title K2Node_AddDelegate}
  ]}
 SelectionMadeEvent {nodes [SelectionMadeEvent!() -> (then, Confirmed:bool=~) {is=title K2Node_FunctionEntry}]}
 UpdateChoiceVM {nodes [
   SetFortLockerCategoryGroupVM!(
    execute
    self:WBP Loadout Card Preview MP Object Reference=Get Card_PresetPreview() -> (Card_PresetPreview:WBP Loadout Card Preview MP Object Reference)
    Viewmodel:Fort Locker Category Group VM Object Reference=Reroute Node(InputPin:Fort Locker Category Group VM Object Reference) -> (OutputPin:Fort Locker Category Group VM Object Reference)
   ) -> (then)
   n1 = UpdateChoiceVM!() -> (then, Index:int) {is=title K2Node_FunctionEntry}
   Reroute Node(InputPin="Get (a copy)"(Array:Array of Fort Locker Category Group VM Object References, Dimension 1=@n1) -> (Output:Fort Locker Category Group VM Object Reference))
   n2 = Length(TargetArray:Array of Wildcards)
   Reroute Node(InputPin:bool="integer > integer"(A=@n1, B=int - int(A=@n2))) -> (OutputPin:bool)
  ]}
]
variables [
 BodyText {type=string, default=""}
 DialogData {type=DialogVM_CrBrowserTutorial_OptIn_C, default=null}
 Min Height {type=integer, default=56}
 SelectionMadeEvent {}
]
widgets {
 info {
  inheritedWidgetCount=0
  namedSlotCount=2
  parentClass=/Script/FortniteUI.FortActivatablePanel
  rootWidgetClass=/FortUILibrary/Systems/Dialog/WBP_UIKit_Dialog_Internal.WBP_UIKit_Dialog_Internal_C
  widgetCount=7
 }
 widgets [
  WBP_UIKit_Dialog_Base {
   widgetClassPath=/FortUILibrary/Systems/Dialog/WBP_UIKit_Dialog_Internal
   bIsVariable=true
   widget=WBP_UIKit_Dialog_Base
   properties {
    useImage=true
    buttons {
     each Button Horizontal Fill=false
     each Button Min Width=200
     each Button Size=Medium
     focuse Button=0
     space Between Buttons=24
     cTA Buttons Types [Regular,Quiet]
     group Placement {
      horizontal Alignment=HAlign_Fill
      vertical Alignment=VAlign_Bottom
      grid {column=1, column Span=0, layer=0, row=2, row Span=0, nudge=<0 0>}
      padding=<ltrb 80 64 80 48>
     }
    }
    dialogSIze {
     height Override {enable=true, value=764}
     max Aspect Ratio {enable=false, value=0}
     max Desired Height {enable=true, value=800}
     max Desired Width {enable=false, value=0}
     min Aspect Ratio {enable=false, value=0}
     min Desired Height {enable=true, value=700}
     min Desired Width {enable=false, value=0}
     width Override {enable=true, value=1424}
    }
    imageProperties {
     material=/FortUILibrary/UIKit/Dialogs/MaterialInstances/MI_UI_DialogBackground
     image Size=<600 0>
     placement {
      horizontal Alignment=HAlign_Fill
      vertical Alignment=VAlign_Fill
      grid {column=1, column Span=0, layer=20, row=0, row Span=1, nudge=<0 0>}
      padding=<ltrb 0 0 0 0>
     }
    }
    titleImagePadding=<ltrb 0 0 8 0>
   }
  }
  SizeBox_3 {widgetClassPath=&SizeBox, namedSlotHost=WBP_UIKit_Dialog_Base, widget=SizeBox_3}
  Overlay {
   widgetClassPath=&Overlay
   bIsVariable=true
   parent=SizeBox_3
   widget=Overlay
   slot {class=SizeBoxSlot, properties {padding=<ltrb 985 0 0 0>}}
  }
  Overlay_0 {widgetClassPath=&Overlay, namedSlotHost=WBP_UIKit_Dialog_Base, widget=Overlay_0}
  ScaleBox_0 {
   widgetClassPath=&ScaleBox
   parent=Overlay_0
   widget=ScaleBox_0
   properties {stretch=ScaleToFitY, userSpecifiedScale=1.1}
   slot {class=OverlaySlot, properties {horizontalAlignment=HAlign_Center, verticalAlignment=VAlign_Center, padding=<ltrb 412 0 0 42>}}
  }
  SizeBox_0 {widgetClassPath=&SizeBox, parent=ScaleBox_0, widget=SizeBox_0, properties {heightOverride=709.8702}, slot {class=ScaleBoxSlot}}
  Image {
   widgetClassPath=&Image
   bIsVariable=true
   parent=SizeBox_0
   widget=Image
   properties {brush {
     bIsDynamicallyLoaded=false
     drawAs=Image
     imageType=FullColor
     mirroring=NoMirror
     resourceName=None
     resourceObject=/OfferCatalog/Textures/BattleRoyale/FeaturedItems/Outfit/T-AthenaSoldiers-CID-904-Athena-Commando-M-Jekyll_03
     tiling=NoTile
     imageSize=<850 850>
     margin=<ltrb 0 0 0 0>
     outlineSettings {
      bUseBrushTransparency=false
      roundingType=HalfHeightRadius
      width=0
      color {colorUseRule=UseColor_Specified, specifiedColor=<rgba 0 0 0 0>}
      cornerRadii=<0 0 0 0>
     }
     tintColor {colorUseRule=UseColor_Specified, specifiedColor=<rgba 1 1 1 1>}
     uVRegion {bIsValid=false, max=<0 0>, min=<0 0>}
    }}
   slot {class=SizeBoxSlot, properties {horizontalAlignment=HAlign_Center, verticalAlignment=VAlign_Center}}
  }
 ]
}
